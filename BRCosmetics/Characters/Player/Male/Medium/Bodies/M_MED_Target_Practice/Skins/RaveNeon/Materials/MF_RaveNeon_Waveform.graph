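# Material function for the RaveNeon player skin that outputs an animated, soft-edged waveform mask.
# A time-driven sine along U, scaled by amplitude and range mask inputs, displaces the UVs alongside an offset and a warp, which defaults to panned texture noise.
# Those UVs sample the inverted blue channel of T_RaveNeon_Body_TA, which sets SmoothStep edges, softened by an input, against the range mask clamped to a minimum thickness.
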MaterialFunction /BRCosmetics/Characters/Player/Male/Medium/Bodies/M_MED_Target_Practice/Skins/RaveNeon/Materials/MF_RaveNeon_Waveform
graph {nodes [
  n1 = TextureSample(UVs=Panner() {constCoordinate=1, speedX=0.2}) {
   samplerSource=SSM_Wrap_WorldGroupSettings
   samplerType=SAMPLERTYPE_LinearColor
   texture=/BRCosmetics/Characters/Player/Male/Medium/Bodies/M_MED_Target_Practice/Skins/RaveNeon/Textures/T_RaveNeon_Body_TA
  }
  n2 = Multiply(A=ComponentMask(None=@n1.RGB) {g=true, r=true}) {constB=0.1}
  n3 = FunctionInput() {bUsePreviewValueAsDefault=true, inputName=WaveOffset, inputType=FunctionInput_Vector2}
  n4 = Add(A=ComponentMask(None=@n3) {r=true}, B=NamedRerouteUsage())
  n5 = FunctionInput() {bUsePreviewValueAsDefault=true, inputName=WaveSpeed, inputType=FunctionInput_Scalar, previewValue=<-0.75 0 0 1>}
  n6 = Frac(None=Multiply(A=Time(), B=@n5) {constB=-1.5})
  n7 = FunctionInput(Preview=TextureCoordinate()) {inputName=UV, inputType=FunctionInput_Vector2}
  n8 = FunctionInput() {bUsePreviewValueAsDefault=true, inputName=WaveNum, inputType=FunctionInput_Scalar, previewValue=<4 0 0 1>}
  n9 = Multiply(A=ComponentMask(None=@n7) {r=true}, B=@n8) {constB=4}
  n10 = Sine(None=Add(A=Add(A=@n6, B=@n9), B=ComponentMask(None=@n3) {g=true}))
  n11 = FunctionInput() {bUsePreviewValueAsDefault=true, inputName=WaveAmplitude, inputType=FunctionInput_Scalar, previewValue=<0.2 0 0 1>}
  n12 = FunctionInput() {bUsePreviewValueAsDefault=true, inputName=WaveRangeMask, inputType=FunctionInput_Scalar, previewValue=<1 0 0 1>}
  n13 = Add(A=Multiply(A=Multiply(A=@n10, B=@n11) {constB=0.2}, B=@n12), B=NamedRerouteUsage())
  n14 = TextureSample(UVs=Add(
    A=FunctionInput(Preview=@n2) {bUsePreviewValueAsDefault=true, inputName=UVWarp, inputType=FunctionInput_Vector2}
    B=MaterialFunctionCall(X=@n4, Y=@n13) {materialFunction=&MakeFloat2}.Result
   )) {
   samplerSource=SSM_Clamp_WorldGroupSettings
   samplerType=SAMPLERTYPE_LinearColor
   texture=&T_RaveNeon_Body_TA
  }
  n15 = OneMinus(None=@n14.B)
  n16 = FunctionInput() {bUsePreviewValueAsDefault=true, inputName=WaveEdgeSoftness, inputType=FunctionInput_Scalar, previewValue=<0.1 0 0 1>}
  n17 = Add(A=@n15, B=@n16) {constB=0.1}
  n18 = FunctionInput() {bUsePreviewValueAsDefault=true, inputName=WaveThickness, inputType=FunctionInput_Scalar, previewValue=<0.15 0 0 1>}
  FunctionOutput(None=SmoothStep(Min=@n15, Max=Add(A=@n17) {constB=0}, Value=Clamp(None=@n12, Min=@n18) {minDefault=0.01}) {constMax=0.1})
  n19 = MaterialFunctionCall(Float2=@n7) {materialFunction=&BreakOutFloat2Components}
  NamedRerouteDeclaration(None=@n19.G) {name=v, nodeColor=<rgba 1 0 0.6562 1>}
  NamedRerouteDeclaration(None=@n19.R) {name=u, nodeColor=<rgba 0 0.9766 1 1>}
 ]}
