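# Material function for the RaveNeon skin of Target Practice that adds animated, audio-reactive neon waveforms to a material's emissive, masked by a TechArtMask channel.
# Three MF_RaveNeon_Waveform layers blend three color parameters, two of them warped by a panning texture.
# A second-UV tile mask marks the record region, which has spinning radial UVs alongside the centered-scale UVs.
# MF_AudioReactivity amplitude is shaped into a ReactiveAmp signal for the effect.
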MaterialFunction /BRCosmetics/Characters/Player/Male/Medium/Bodies/M_MED_Target_Practice/Skins/RaveNeon/Materials/MF_RaveNeon
graph {nodes [
  n1 = GetMaterialAttributes(None=FunctionInput() {inputName=Material, inputType=FunctionInput_MaterialAttributes}) {attributeGetTypes [#1]}
  n2 = FunctionInput() {inputName=TechArtMask, sortPriority=1}
  n3 = Multiply(
   A=Multiply(
    A=ChannelMaskParameter(None=@n2) {group=RaveNeon, maskChannel=Green, parameterName=RN_TAChannel, defaultValue=<rgba 0 1 0 0>}.R
    B=NamedRerouteUsage()
   )
   B=LinearInterpolate(B=Constant() {r=3}, Alpha=NamedRerouteUsage()) {constA=1, constB=1.5}
  )
  FunctionOutput(None=SetMaterialAttributes!(MaterialAttributes=@n1.MaterialAttributes, EmissiveColor=Add(A=@n1.EmissiveColor, B=@n3)) {attributeSetTypes [#1]})
  n4 = ScalarParameter() {defaultValue=-0.75, group=RaveNeon, parameterName=WaveSpeed}
  LinearInterpolate(
   A=@n4
   B=ScalarParameter() {defaultValue=-2, group=RaveNeon, parameterName=WaveSpeed_Reactive}
   Alpha=ScalarParameter() {group=RaveNeon, parameterName=RN_EmoteReactivity}
  ) {constB=0.001}
  n5 = VectorParameter() {group=RaveNeon, parameterName=RN_Color01, defaultValue=<rgba 0.4019 0.0628 0.5417 1>}
  n6 = NamedRerouteUsage()
  n7 = TextureSample(UVs=Panner(Speed=LinearInterpolate(Alpha=@n6) {constA=-0.3, constB=-0.6}) {constCoordinate=1, speedX=-0.3}) {
   samplerSource=SSM_Wrap_WorldGroupSettings
   samplerType=SAMPLERTYPE_LinearColor
   texture=/BRCosmetics/Characters/Player/Male/Medium/Bodies/M_MED_Target_Practice/Skins/RaveNeon/Textures/T_RaveNeon_Body_TA
  }
  n8 = ComponentMask(None=@n7.RGB) {g=true, r=true}
  n9 = MaterialFunctionCall!(
   WaveAmplitude=Multiply(A=NamedRerouteUsage(), B=Constant() {r=0.7})
   UV=Add(A=NamedRerouteUsage(), B=Constant2Vector() {g=-0.3})
   WaveNum=Constant() {r=5}
   WaveOffset=Constant2Vector() {g=-0.3}
   WaveSpeed=NamedRerouteUsage()
   WaveRangeMask=Multiply(A=NamedRerouteUsage()) {constB=0.6667}
   WaveThickness=Multiply(A=NamedRerouteUsage()) {constB=10}
   WaveEdgeSoftness=NamedRerouteUsage()
   UVWarp=Multiply(A=@n8) {constB=0.5}
  ) {materialFunction=&/MF_RaveNeon_Waveform}
  n10 = LinearInterpolate(
   A=VectorParameter() {group=RaveNeon, parameterName=RN_Color02, defaultValue=<rgba 0.0513 0.3712 1 1>}.RGB
   B=VectorParameter() {group=RaveNeon, parameterName=RN_Color03, defaultValue=<rgba 0.1441 0.9216 0.3813 1>}.RGB
   Alpha=@n9.Result
  )
  n11 = NamedRerouteUsage()
  n12 = Constant2Vector() {g=-0.1}
  n13 = NamedRerouteUsage()
  n14 = Add(
   A=MaterialFunctionCall(
    WaveAmplitude=NamedRerouteUsage()
    UV=Add(A=@n11, B=@n12)
    WaveNum=Constant() {r=3}
    WaveOffset=Constant2Vector() {g=-0.5}
    WaveSpeed=NamedRerouteUsage()
    WaveRangeMask=Multiply(A=@n13) {constB=0.95}
    WaveThickness=Constant() {r=0.9}
    WaveEdgeSoftness=NamedRerouteUsage()
    UVWarp=Multiply(A=@n8) {constB=0.4}
   ) {materialFunction=&MF_RaveNeon_Waveform}.Result
   B=@n9.Result
  )
  n15 = NamedRerouteUsage()
  n16 = NamedRerouteUsage()
  n17 = NamedRerouteUsage()
  n18 = NamedRerouteUsage()
  n19 = NamedRerouteUsage()
  n20 = NamedRerouteUsage()
  n21 = NamedRerouteUsage()
  NamedRerouteDeclaration(None=LinearInterpolate(
    B=LinearInterpolate(A=@n5.RGB, B=@n10, Alpha=@n14)
    Alpha=MaterialFunctionCall(WaveAmplitude=@n15, UV=@n16, WaveNum=@n17, WaveSpeed=@n18, WaveRangeMask=@n19, WaveThickness=@n20, WaveEdgeSoftness=@n21) {materialFunction=&MF_RaveNeon_Waveform}.Result
   )) {name=Emissive_OUT, nodeColor=<rgba 0 0.25 1 1>}
  NamedRerouteDeclaration(None=ScalarParameter() {defaultValue=4, group=RaveNeon, parameterName=WaveNum}) {name=WaveNum, nodeColor=<rgba 0 0.9297 1 1>}
  n22 = ScalarParameter() {defaultValue=0.15, group=RaveNeon, parameterName=WaveAmp}
  n23 = NamedRerouteUsage()
  NamedRerouteDeclaration(None=LinearInterpolate(A=Add(A=@n22, B=@n23), B=LinearInterpolate(Alpha=@n23) {constA=0.35, constB=0}, Alpha=NamedRerouteUsage())) {name=WaveAmp, nodeColor=<rgba 1 0.0469 0 1>}
  NamedRerouteDeclaration(None=LinearInterpolate(
    A=ScalarParameter() {defaultValue=0.15, group=RaveNeon, parameterName=WaveThickness}
    B=ScalarParameter() {defaultValue=0.8, group=RaveNeon, parameterName=WaveThickness_Record}
    Alpha=NamedRerouteUsage()
   )) {name=WaveThickness, nodeColor=<rgba 0.3359 1 0 1>}
  NamedRerouteDeclaration(None=ScalarParameter() {group=RaveNeon, parameterName=WaveSoftness}) {name=WaveEdgeSoftness, nodeColor=<rgba 0.5703 1 0 1>}
  NamedRerouteDeclaration() {name=Debug, nodeColor=<rgba 0 1 0.2266 1>}
  n24 = TextureCoordinate() {coordinateIndex=1}
  n25 = Frac(None=@n24)
  n26 = ScalarParameter() {defaultValue=0.7, group=RaveNeon, parameterName=RN_GlobalScale}
  n27 = TextureSample() {
   constCoordinate=1
   samplerSource=SSM_Wrap_WorldGroupSettings
   samplerType=SAMPLERTYPE_LinearColor
   texture=&T_RaveNeon_Body_TA
  }
  n28 = Multiply(A=ComponentMask(None=@n27.RGB) {g=true, r=true}) {constB=0.7}
  n29 = MaterialFunctionCall(Float2=MaterialFunctionCall(Vector or UVs=Add(A=@n25) {constB=-0.5}) {materialFunction=&VectorToRadialValue}.Radial Coordinates) {materialFunction=&BreakOutFloat2Components}
  n30 = NamedRerouteDeclaration(None=Multiply(A=@n29.G) {constB=2}) {name=UV_G_Record, nodeColor=<rgba 0.2656 0 1 1>}
  n31 = OneMinus(None=Saturate(None=Power(Base=@n30) {constExponent=0.3}))
  n32 = Multiply(A=@n29.R) {constB=2}
  n33 = Multiply(A=Time(), B=ScalarParameter() {defaultValue=0.05, group=RaveNeon, parameterName=RN_RecordSpinSpeed})
  n34 = Frac(None=Add(A=Add(A=@n32) {constB=-0.5}, B=Frac(None=@n33)))
  n35 = MaterialFunctionCall(Float2=Add(
    A=Multiply(A=@n28, B=@n31)
    B=MaterialFunctionCall(X=@n30, Y=@n34) {materialFunction=&MakeFloat2}.Result
   )) {materialFunction=&BreakOutFloat2Components}
  n36 = Subtract(A=NamedRerouteUsage(), B=Constant() {r=0.5})
  n37 = Constant() {r=0.2}
  n38 = Frac(None=Add(A=@n35.R, B=Multiply(A=@n36, B=@n37) {constB=2}))
  NamedRerouteDeclaration(None=LinearInterpolate(
    A=MaterialFunctionCall(UVs=@n25, Texture Scale=@n26) {materialFunction=&ScaleUVsByCenter}.UVs
    B=MaterialFunctionCall(X=@n38, Y=@n35.G) {materialFunction=&MakeFloat2}.Result
    Alpha=NamedRerouteUsage()
   )) {name=UVs, nodeColor=<rgba 1 0 0.5859 1>}
  n39 = MaterialFunctionCall(Float2=NamedRerouteUsage()) {materialFunction=&BreakOutFloat2Components}
  n40 = OneMinus(None=Distance(A=@n39.R, B=Constant() {r=0.5}))
  NamedRerouteDeclaration(None=LinearInterpolate(A=SmoothStep(Value=@n40) {constMin=0.6}, B=NamedRerouteUsage(), Alpha=NamedRerouteUsage())) {name=WaveformRangeMask, nodeColor=<rgba 0.7344 1 0 1>}
  NamedRerouteDeclaration(None=@n39.R) {name=U, nodeColor=<rgba 0.9453 0 1 1>}
  NamedRerouteDeclaration(None=@n39.G) {name=V, nodeColor=<rgba 0 1 0.625 1>}
  n41 = NamedRerouteUsage()
  n42 = MaterialFunctionCall() {materialFunction=/Game/Packages/Fortress_SharedMaterials/Base_Material_Functions/BaseCharacterShaderFunctions/Audio/MF_AudioReactivity}
  n43 = Constant() {r=7}
  NamedRerouteDeclaration(None=SmoothStep(Max=LinearInterpolate(Alpha=@n41) {constA=2.5, constB=1.3}, Value=Power(Base=@n42.Amplitude, Exp=@n43) {constExponent=3}) {constMax=0.5}) {name=ReactiveAmp, nodeColor=<rgba 0 1 0.4609 1>}
  NamedRerouteDeclaration(None=@n4) {name=WaveSpeed, nodeColor=<rgba 0 1 0.7891 1>}
  n44 = Add(A=MaterialFunctionCall(Float2=@n24) {materialFunction=&BreakOutFloat2Components}.R) {constB=-1}
  NamedRerouteDeclaration(None=Saturate(None=Floor(None=@n44))) {name=Mask_Record, nodeColor=<rgba 0.2891 0 1 1>}
  NamedRerouteUsage()
  NamedRerouteUsage()
  NamedRerouteUsage()
  NamedRerouteUsage()
  NamedRerouteUsage()
 ]}
